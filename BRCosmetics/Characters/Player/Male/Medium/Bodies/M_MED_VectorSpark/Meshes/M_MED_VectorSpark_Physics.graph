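# Physics asset for the VectorSpark medium male player body cosmetic, set up for dynamic shirt panels around the hips and legs.
# Capsules on the pelvis, thighs and calves form the lower-body shapes, linked by free angular joints.
# Six left/right front, back and side panels are thin boxes, each jointed to a tiny master-bone capsule that links freely to the pelvis.
# Panel joints carry narrow swing and zero twist limits, tightest on the sides.
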PhysicsAsset /BRCosmetics/Characters/Player/Male/Medium/Bodies/M_MED_VectorSpark/Meshes/M_MED_VectorSpark_Physics
bodies {
 calf_l [{shapeType=Capsule, length=26.2475, radius=7.5065, shapeName=calf_l_capsule, center=<-15.4905 -2.5784 0>, rotation=<rot -90 0 0>}]
 calf_r [{shapeType=Capsule, length=26.6202, radius=7.573, shapeName=calf_r_capsule, center=<15.4905 2.4423 0>, rotation=<rot 90 0 0>}]
 dyn_shirt_back_01_l [{
   shapeType=Box
   extentX=18
   extentY=1.2793
   extentZ=18
   shapeName=dyn_shirt_back_01_l_box
   center=<4.3269 -2.0007 -0.1443>
   rotation=<rot 0 0.0001 -6.9067>
  }]
 dyn_shirt_back_01_r [{
   shapeType=Box
   extentX=18
   extentY=1.2793
   extentZ=18
   shapeName=dyn_shirt_back_01_r_box
   center=<-4.3269 1.9021 0.1218>
   rotation=<rot 0 -180 -173.6506>
  }]
 dyn_shirt_back_mstr_l [{shapeType=Capsule, length=0.505, radius=0.505, shapeName=dyn_shirt_back_mstr_l_capsule, rotation=<rot 0 -90 0>}]
 dyn_shirt_back_mstr_r [{shapeType=Capsule, length=0.505, radius=0.505, shapeName=dyn_shirt_back_mstr_r_capsule, rotation=<rot 0 -90 0>}]
 dyn_shirt_front_01_l [{shapeType=Box, extentX=18, extentY=1.0151, extentZ=18, shapeName=dyn_shirt_front_01_l_box, center=<4.3269 -0.3394 0>}]
 dyn_shirt_front_01_r [{
   shapeType=Box
   extentX=18
   extentY=0.8383
   extentZ=18
   shapeName=dyn_shirt_front_01_r_box
   center=<-4.3269 0.4105 0>
   rotation=<rot 0 180 180>
  }]
 dyn_shirt_front_mstr_l [{shapeType=Capsule, length=0.505, radius=0.505, shapeName=dyn_shirt_front_mstr_l_capsule, rotation=<rot 0 -90 0>}]
 dyn_shirt_front_mstr_r [{shapeType=Capsule, length=0.505, radius=0.505, shapeName=dyn_shirt_front_mstr_r_capsule, rotation=<rot 0 -90 0>}]
 dyn_shirt_side_01_l [{shapeType=Box, extentX=18, extentY=1.0151, extentZ=18, shapeName=dyn_shirt_side_01_l_box, center=<4.3269 -1.6686 0>}]
 dyn_shirt_side_01_r [{
   shapeType=Box
   extentX=18
   extentY=0.8383
   extentZ=18
   shapeName=dyn_shirt_side_01_r_box
   center=<-4.3269 1.4865 0>
   rotation=<rot 0 180 180>
  }]
 dyn_shirt_side_mstr_l [{shapeType=Capsule, length=0.505, radius=0.505, shapeName=dyn_shirt_side_mstr_l_capsule, rotation=<rot 0 -90 0>}]
 dyn_shirt_side_mstr_r [{shapeType=Capsule, length=0.505, radius=0.505, shapeName=dyn_shirt_side_mstr_r_capsule, rotation=<rot 0 -90 0>}]
 pelvis [{shapeType=Capsule, length=10.592, radius=11.3024, shapeName=pelvis_capsule, center=<0 -1.6149 0>}]
 thigh_l [{
   shapeType=Capsule
   length=32.5333
   radius=10.0202
   shapeName=thigh_l_capsule
   center=<-18.0679 -0.701 -0.1882>
   rotation=<rot -90 -179.944 -179.8875>
  }]
 thigh_r [{shapeType=Capsule, length=32.5552, radius=10.0202, shapeName=thigh_r_capsule, center=<18.0904 0.8638 0.0231>, rotation=<rot 90 0 0.2083>}]
}
constraints [
 {
  bone1Name=thigh_l
  bone2Name=pelvis
  swing1LimitDegrees=45
  swing1Motion=Free
  swing2LimitDegrees=45
  swing2Motion=Free
  twistLimitDegrees=45
  twistMotion=Free
 }
 {
  bone1Name=calf_l
  bone2Name=thigh_l
  swing1LimitDegrees=45
  swing1Motion=Free
  swing2LimitDegrees=45
  swing2Motion=Free
  twistLimitDegrees=45
  twistMotion=Free
 }
 {
  bone1Name=thigh_r
  bone2Name=pelvis
  swing1LimitDegrees=45
  swing1Motion=Free
  swing2LimitDegrees=45
  swing2Motion=Free
  twistLimitDegrees=45
  twistMotion=Free
 }
 {
  bone1Name=calf_r
  bone2Name=thigh_r
  swing1LimitDegrees=45
  swing1Motion=Free
  swing2LimitDegrees=45
  swing2Motion=Free
  twistLimitDegrees=45
  twistMotion=Free
 }
 {
  bone1Name=dyn_shirt_back_mstr_l
  bone2Name=pelvis
  swing1LimitDegrees=45
  swing1Motion=Free
  swing2LimitDegrees=45
  swing2Motion=Free
  twistLimitDegrees=45
  twistMotion=Free
 }
 {
  bone1Name=dyn_shirt_back_mstr_r
  bone2Name=pelvis
  swing1LimitDegrees=45
  swing1Motion=Free
  swing2LimitDegrees=45
  swing2Motion=Free
  twistLimitDegrees=45
  twistMotion=Free
 }
 {
  bone1Name=dyn_shirt_front_mstr_l
  bone2Name=pelvis
  swing1LimitDegrees=45
  swing1Motion=Free
  swing2LimitDegrees=45
  swing2Motion=Free
  twistLimitDegrees=45
  twistMotion=Free
 }
 {
  bone1Name=dyn_shirt_front_mstr_r
  bone2Name=pelvis
  swing1LimitDegrees=45
  swing1Motion=Free
  swing2LimitDegrees=45
  swing2Motion=Free
  twistLimitDegrees=45
  twistMotion=Free
 }
 {
  bone1Name=dyn_shirt_side_mstr_l
  bone2Name=pelvis
  swing1LimitDegrees=45
  swing1Motion=Free
  swing2LimitDegrees=45
  swing2Motion=Free
  twistLimitDegrees=45
  twistMotion=Free
 }
 {
  bone1Name=dyn_shirt_side_mstr_r
  bone2Name=pelvis
  swing1LimitDegrees=45
  swing1Motion=Free
  swing2LimitDegrees=45
  swing2Motion=Free
  twistLimitDegrees=45
  twistMotion=Free
 }
 {bone1Name=dyn_shirt_back_01_l, bone2Name=dyn_shirt_back_mstr_l, swing1LimitDegrees=45, swing2LimitDegrees=5, twistLimitDegrees=0}
 {bone1Name=dyn_shirt_back_01_r, bone2Name=dyn_shirt_back_mstr_r, swing1LimitDegrees=45, swing2LimitDegrees=5, twistLimitDegrees=0}
 {bone1Name=dyn_shirt_front_01_l, bone2Name=dyn_shirt_front_mstr_l, swing1LimitDegrees=45, swing2LimitDegrees=5, twistLimitDegrees=0}
 {bone1Name=dyn_shirt_front_01_r, bone2Name=dyn_shirt_front_mstr_r, swing1LimitDegrees=45, swing2LimitDegrees=5, twistLimitDegrees=0}
 {bone1Name=dyn_shirt_side_01_l, bone2Name=dyn_shirt_side_mstr_l, swing1LimitDegrees=16, swing2LimitDegrees=2, twistLimitDegrees=0}
 {bone1Name=dyn_shirt_side_01_r, bone2Name=dyn_shirt_side_mstr_r, swing1LimitDegrees=16, swing2LimitDegrees=2, twistLimitDegrees=0}
]
